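# Physics asset for the F_MED_Farmer_Backpack cosmetic, giving its flower and bowknot decorations physics bodies and joint constraints.
# A box covers the backpack root and capsules cover the flower, six petals and bow segments.
# The flower and its separate root bone attach to the backpack, with petals constrained to the flower with small swing limits.
# The bowknot hangs off the flower root, its first two segments per side joining the knot and each third chaining from the second.
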PhysicsAsset /BRCosmetics/Accessories/FORT_Backpacks/F_MED_Farmer_Backpack/Meshes/F_MED_Farmer_Backpack_PhysicsAsset
bodies {
 backpack_root [{shapeType=Box, extentX=31.5, extentY=7.5, extentZ=38.5, center=<-9.3379 3.5376 0.0001>, rotation=<rot 0 13.2114 0>}]
 dny_bowknot_l_1 [{shapeType=Capsule, length=2.1751, radius=1.705, center=<-0.2681 -1.7835 0.2061>, rotation=<rot -80.1903 121.4316 147.6934>}]
 dny_bowknot_l_2 [{shapeType=Capsule, length=1.1312, radius=1.3656, center=<1.5176 -0.6025 0.3047>, rotation=<rot -76.8538 -39.3189 14.5886>}]
 dny_bowknot_l_3 [{
   shapeType=Capsule
   length=1.1928
   radius=1.3926
   shapeName=dny_bowknot_l_3_capsule
   center=<1.7196 -0.6656 1.3102>
   rotation=<rot 46.7365 156.4971 -168.0452>
  }]
 dny_bowknot_r_1 [{shapeType=Capsule, length=1.7846, radius=1.75, center=<1.0942 1.8847 1.1976>, rotation=<rot -77.9219 95.4879 -55.3592>}]
 dny_bowknot_r_2 [{shapeType=Capsule, length=1.531, radius=1.1499, center=<1.3782 0.0473 0.3397>, rotation=<rot -83.4264 0.554 3.8767>}]
 dny_bowknot_r_3 [{shapeType=Capsule, length=0.7315, radius=1.3178, shapeName=dny_bowknot_r_3_capsule, center=<1.952 0 0.9502>, rotation=<rot 77.4855 180 180>}]
 dny_bowknot_root [{shapeType=Capsule, length=1, radius=2.25, center=<0.2733 0.1546 1.4201>, rotation=<rot -80.8175 0.5544 70.2817>}]
 dny_floew_root [{shapeType=Capsule, length=7.3876, radius=2.126, center=<3.0196 -0.7309 0.8237>, rotation=<rot 0 74.8921 -84.0322>}]
 dny_petal_1 [{shapeType=Capsule, length=3.25, radius=2.75, center=<-1.7953 -0.2028 -0.0213>, rotation=<rot -89.6093 179.9996 -179.9995>}]
 dny_petal_2 [{shapeType=Capsule, length=3.25, radius=2.75, center=<-2.4376 -0.2265 -0.0102>, rotation=<rot -89.622 179.9999 180>}]
 dny_petal_3 [{shapeType=Capsule, length=3.25, radius=2.75, center=<-2.1175 0.0203 -0.0337>, rotation=<rot -89.5335 -179.9999 -180>}]
 dny_petal_4 [{
   shapeType=Capsule
   length=3.4075
   radius=2.5235
   shapeName=dny_petal_4_capsule
   center=<-1.3704 0.1446 2.9671>
   rotation=<rot 0 -88.8285 -74.3359>
  }]
 dny_petal_5 [{shapeType=Capsule, length=3.25, radius=2.75, center=<-2.2264 -0.8306 0.5719>, rotation=<rot -70.1173 179.9999 179.9999>}]
 dny_petal_6 [{shapeType=Capsule, length=3.25, radius=2.75, center=<-1.9308 0 0.0091>, rotation=<rot -89.7252 -179.9996 179.9998>}]
 floew [{shapeType=Capsule, length=1.25, radius=6.2066, center=<0.0761 -0.0854 0.9692>, rotation=<rot 0.4839 -94.874 -90>}]
}
constraints [
 {
  bone1Name=dny_floew_root
  bone2Name=backpack_root
  swing1LimitDegrees=0.05
  swing2LimitDegrees=0.05
  swing2Motion=Locked
  twistLimitDegrees=45
  twistMotion=Locked
 }
 {bone1Name=floew, bone2Name=backpack_root, swing1LimitDegrees=1, swing2LimitDegrees=1, twistLimitDegrees=45, twistMotion=Locked}
 {
  bone1Name=dny_bowknot_root
  bone2Name=dny_floew_root
  swing1LimitDegrees=0.05
  swing2LimitDegrees=0.05
  twistLimitDegrees=45
  twistMotion=Locked
 }
 {
  bone1Name=dny_bowknot_l_1
  bone2Name=dny_bowknot_root
  swing1LimitDegrees=20
  swing2LimitDegrees=30
  twistLimitDegrees=45
  twistMotion=Locked
 }
 {bone1Name=dny_bowknot_l_2, bone2Name=dny_bowknot_root, swing1LimitDegrees=15.76, swing2LimitDegrees=40, twistLimitDegrees=5}
 {
  bone1Name=dny_bowknot_r_1
  bone2Name=dny_bowknot_root
  swing1LimitDegrees=10
  swing2LimitDegrees=20
  twistLimitDegrees=45
  twistMotion=Locked
 }
 {bone1Name=dny_bowknot_r_2, bone2Name=dny_bowknot_root, swing1LimitDegrees=10, swing2LimitDegrees=40, twistLimitDegrees=5}
 {bone1Name=dny_petal_1, bone2Name=floew, swing1LimitDegrees=6, swing2LimitDegrees=8, twistLimitDegrees=45, twistMotion=Locked}
 {bone1Name=dny_petal_2, bone2Name=floew, swing1LimitDegrees=6, swing2LimitDegrees=8, twistLimitDegrees=45, twistMotion=Locked}
 {bone1Name=dny_petal_3, bone2Name=floew, swing1LimitDegrees=6, swing2LimitDegrees=8, twistLimitDegrees=45, twistMotion=Locked}
 {bone1Name=dny_petal_4, bone2Name=floew, swing1LimitDegrees=6, swing2LimitDegrees=8, twistLimitDegrees=45, twistMotion=Locked}
 {bone1Name=dny_petal_5, bone2Name=floew, swing1LimitDegrees=6, swing2LimitDegrees=8, twistLimitDegrees=45, twistMotion=Locked}
 {bone1Name=dny_petal_6, bone2Name=floew, swing1LimitDegrees=6, swing2LimitDegrees=8, twistLimitDegrees=45, twistMotion=Locked}
 {bone1Name=dny_bowknot_l_3, bone2Name=dny_bowknot_l_2, swing1LimitDegrees=25, swing2LimitDegrees=30, twistLimitDegrees=5}
 {bone1Name=dny_bowknot_r_3, bone2Name=dny_bowknot_r_2, swing1LimitDegrees=25, swing2LimitDegrees=30, twistLimitDegrees=5}
]
