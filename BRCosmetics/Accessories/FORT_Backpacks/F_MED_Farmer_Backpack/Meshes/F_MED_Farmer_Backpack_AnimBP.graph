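# Animation Blueprint for the Farmer backpack accessory, adding spine-driven rotation and secondary physics motion.
# Each update maps the yaw of spine_05 relative to spine_01, measured from ref pose on the owner skeletal mesh component, to a clamped backpack yaw.
# That rotation and a small offset are added to backpack_root, then a RigidBody node simulates in that bone's space.
# Two spring-constrained AnimDynamics nodes finish with dny_eye_l and dny_eye_r jiggle.
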AnimBlueprint /BRCosmetics/Accessories/FORT_Backpacks/F_MED_Farmer_Backpack/Meshes/F_MED_Farmer_Backpack_AnimBP
target_skeleton=&/F_MED_Farmer_Backpack_Skeleton
graphs [
 AnimGraph {nodes [
   AnimGraphNode_SubInput_0 = Input Pose()
   n1 = "Transform (Modify) Bone - Bone: backpack_root"(
    ComponentPose=Local To Component(LocalPose=@AnimGraphNode_SubInput_0)
    Alpha=1
    Translation="0,-0.5,0"
    Rotation=Get backpack_rotate()
    Scale="1,1,1"
   ) {
    rotationMode=BMM_Additive
    rotationSpace=BCS_BoneSpace
    translationMode=BMM_Additive
    translationSpace=BCS_BoneSpace
    alphaBoolBlend {blendOption=Linear}
    boneToModify {boneName=backpack_root}
    translation {y=-0.5}
   }
   n2 = RigidBody(ComponentPose=@n1) {
    bOverrideWorldGravity=true
    lODThreshold=2
    overridePhysicsAsset=&/F_MED_Farmer_Backpack_PhysicsAsset
    simulationSpace=BaseBoneSpace
    alphaBoolBlend {blendOption=Linear}
    baseBoneRef {boneName=backpack_root}
    componentAppliedLinearAccClamp=<1500 1500 1500>
    componentLinearAccScale=<0.5 0.5 0.5>
    componentLinearVelScale=<0.5 0.5 0.5>
    overrideWorldGravity {z=-980}
   }
   n3 = Anim Dynamics - Bone: dny_eye_l(ComponentPose=@n2, Alpha=1, ExternalForce="0,0,0") {
    angularDampingOverride=0.999
    angularSpringConstant=1000
    bAngularSpring=true
    bGravityOverrideInSimSpace=true
    bOverrideAngularDamping=true
    bOverrideLinearDamping=true
    bUseGravityOverride=true
    gravityScale=0.3
    linearDampingOverride=0.999
    alphaBoolBlend {blendOption=Linear}
    boundBone {boneName=dny_eye_l}
    componentAppliedLinearAccClamp=<1000 1000 1000>
    componentLinearAccScale=<1 1 1>
    componentLinearVelScale=<1 1 1>
    gravityOverride {z=-700}
    physicsBodyDefinitions [{
      sphereCollisionRadius=0
      boundBone {boneName=dny_eye_l}
      boxExtents=<5 5 1>
      constraintSetup {angularTargetAxis=AxisY, angularLimitsMax=<5 15 0>, angularLimitsMin=<-5 -15 0>, angularTarget=<0 0 0>}
      localJointOffset=<1 -0.1 -0.5>
     }]
   }
   n4 = Anim Dynamics - Bone: dny_eye_r(ComponentPose=@n3, Alpha=1, ExternalForce="0,0,0") {
    angularDampingOverride=0.999
    angularSpringConstant=1000
    bAngularSpring=true
    bGravityOverrideInSimSpace=true
    bOverrideAngularDamping=true
    bOverrideLinearDamping=true
    bUseGravityOverride=true
    gravityScale=0.3
    linearDampingOverride=0.999
    alphaBoolBlend {blendOption=Linear}
    boundBone {boneName=dny_eye_r}
    componentAppliedLinearAccClamp=<1000 1000 1000>
    componentLinearAccScale=<1 1 1>
    componentLinearVelScale=<1 1 1>
    gravityOverride {z=-700}
    physicsBodyDefinitions [{
      sphereCollisionRadius=0
      boundBone {boneName=dny_eye_r}
      boxExtents=<5 5 1>
      constraintSetup {angularTargetAxis=AxisY, angularLimitsMax=<5 15 0>, angularLimitsMin=<-5 -15 0>, angularTarget=<0 0 0>}
      localJointOffset=<1 -0.1 -0.5>
     }]
   }
   Output Pose(Result=Component To Local(ComponentPose=@n4)) {layerGroup=None}
  ]}
 EventGraph {nodes [
   Event BlueprintUpdateAnimation()
   n1 = GetOwnerSkeletalMeshComponent()
   n2 = BreakRotator(InRot=BreakTransform(InTransform=GetDeltaTransformFromRefPose(self=@n1, BoneName=spine_05, BaseName=spine_01)).Rotation)
   > Set backpack_rotate(backpack_rotate=MakeRotator(Yaw=MapRangeClamped(Value=@n2.Yaw, InRangeA=12, InRangeB=-30, OutRangeA=5, OutRangeB=45)))
  ]}
]
variables [backpack_rotate {type=Rotator}]
